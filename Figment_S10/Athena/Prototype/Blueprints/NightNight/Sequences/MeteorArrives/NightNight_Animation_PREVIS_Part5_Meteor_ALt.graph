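# Alternate previs Level Sequence for part 5 of the NightNight meteor arrival, animating a meteor's fall.
# The spawned BP_NightNight_Meteor_Drop descends on keyed Z, with its particle component toggled and scale collapsing at impact.
# An attached camera rail rig animates its rail position, alongside a SceneOrigin mesh, with no camera or camera-cut track.
# A fade track labeled UnusedFade_KeepIfNeeded holds accelerating flash pulses, and an event track fires two late triggers.
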LevelSequence /Figment_S10/Athena/Prototype/Blueprints/NightNight/Sequences/MeteorArrives/NightNight_Animation_PREVIS_Part5_Meteor_ALt
bindings [
 Camera Rig Rail {
  id=#1
  possessed_class=CameraRig_Rail
  tracks [
   {
    type=MovieSceneSpawnTrack
    sections [{type=MovieSceneSpawnSection, channels [None {type=MovieSceneScriptingBoolChannel, default=false, const [true,1]}]}]
   }
   {
    type=MovieScene3DTransformTrack
    sections [{
      type=MovieScene3DTransformSection
      channels [
       Location.X {type=MovieSceneScriptingDoubleChannel, default=0, const [0,1]}
       Location.Y {type=MovieSceneScriptingDoubleChannel, default=0, const [0,1]}
       Location.Z {type=MovieSceneScriptingDoubleChannel, default=0, const [0,1]}
       Rotation.X {type=MovieSceneScriptingDoubleChannel, default=0, const [0,1]}
       Rotation.Y {type=MovieSceneScriptingDoubleChannel, default=0, const [0,1]}
       Rotation.Z {type=MovieSceneScriptingDoubleChannel, default=0, const [0,1]}
       Scale.X {type=MovieSceneScriptingDoubleChannel, default=1, const [1,1]}
       Scale.Y {type=MovieSceneScriptingDoubleChannel, default=1, const [1,1]}
       Scale.Z {type=MovieSceneScriptingDoubleChannel, default=1, const [1,1]}
      ]
     }]
   }
   {
    type=MovieSceneFloatTrack
    display_name=Current Position on Rail
    sections [{
      type=MovieSceneFloatSection
      channels [None {type=MovieSceneScriptingFloatChannel, default=0.187, keys [[1023,0.187],[2420,0.9365],[2454,0.9397],[2829,0.9516]]}]
     }]
   }
   {type=MovieScene3DAttachTrack, sections [{type=MovieScene3DAttachSection, end_frame=3677, start_frame=1023}]}
  ]
 }
 SceneOrigin {
  id=#2
  possessed_class=FortStaticMeshActor
  tracks [
   {
    type=MovieScene3DTransformTrack
    sections [{
      type=MovieScene3DTransformSection
      channels [
       Location.X {type=MovieSceneScriptingDoubleChannel, default=65949.1797}
       Location.Y {type=MovieSceneScriptingDoubleChannel, default=-42629.9727}
       Location.Z {type=MovieSceneScriptingDoubleChannel, default=411558.9062}
       Rotation.X {type=MovieSceneScriptingDoubleChannel, default=0}
       Rotation.Y {type=MovieSceneScriptingDoubleChannel, default=0}
       Rotation.Z {type=MovieSceneScriptingDoubleChannel, default=0}
       Scale.X {type=MovieSceneScriptingDoubleChannel, default=1}
       Scale.Y {type=MovieSceneScriptingDoubleChannel, default=1}
       Scale.Z {type=MovieSceneScriptingDoubleChannel, default=1}
      ]
     }]
   }
   {type=MovieSceneSpawnTrack, sections [{type=MovieSceneSpawnSection, channels [None {type=MovieSceneScriptingBoolChannel, default=true}]}]}
  ]
 }
 BP_NightNight_Meteor_Drop {
  id=#3
  possessed_class=BP_NightNight_Meteor_Drop_C
  tracks [
   {
    type=MovieSceneSpawnTrack
    sections [{
      type=MovieSceneSpawnSection
      channels [None {type=MovieSceneScriptingBoolChannel, default=false, keys [[1023,false],[1165,true],[3175,false]]}]
     }]
   }
   {
    type=MovieScene3DTransformTrack
    sections [{
      type=MovieScene3DTransformSection
      channels [
       Location.X {type=MovieSceneScriptingDoubleChannel, default=65949.1797, const [65949.1797,1]}
       Location.Y {type=MovieSceneScriptingDoubleChannel, default=-42629.9727, const [-42629.9727,1]}
       Location.Z {
        type=MovieSceneScriptingDoubleChannel
        default=13682.0664
        keys [[1023,263100],[1325,232251.5312],[2732,20017.3359],[2760,18785.25],[3154,13518.7139],[3171,1284.76]]
       }
       Rotation.X {type=MovieSceneScriptingDoubleChannel, default=0}
       Rotation.Y {type=MovieSceneScriptingDoubleChannel, default=0}
       Rotation.Z {type=MovieSceneScriptingDoubleChannel, default=0}
       Scale.X {type=MovieSceneScriptingDoubleChannel, default=0, keys [[3154,1],[3175,0]]}
       Scale.Y {type=MovieSceneScriptingDoubleChannel, default=0, keys [[3154,1],[3175,0]]}
       Scale.Z {type=MovieSceneScriptingDoubleChannel, default=0, keys [[3154,1],[3175,0]]}
      ]
     }]
   }
  ]
 }
 P_NightNight_Meteor_Drop {
  id=#4
  parent_id=#3
  possessed_class=ParticleSystemComponent
  tracks [{
    type=MovieSceneParticleTrack
    display_name=Particle System
    sections [{
      type=MovieSceneParticleSection
      channels [None {type=MovieSceneScriptingParticleChannel, keys [[1170.4525, "<ParticleKey.ACTIVATE: 0>"], [3154, "<ParticleKey.DEACTIVATE: 1>"]]}]
     }]
   }]
 }
]
display_rate {denominator=1, numerator=60}
tick_resolution {denominator=1, numerator=24000}
tracks [
 {
  type=MovieSceneFadeTrack
  display_name=UnusedFade_KeepIfNeeded
  sections [{
    type=MovieSceneFadeSection
    channels [None {
      type=MovieSceneScriptingFloatChannel
      default=0
      keys [[1023,0],[2463,1],[2464,0],[2493,1],[2494,0],[2523,1],[2524,0],[2538,1],[2539,0],[2553,1],[2554,0],[2568,1],[2569,0],[2583,1],[2584,0],[2598,1],[2599,0],[2613,1],[2614,0],[2623,1],[2624,0],[2633,1],[2634,0],[2643,1],[2644,0],[2653,1],[2654,0],[2663,1],[2664,0],[2673,1],[2674,0],[2683,1],[2684,0],[2693,1],[2694,0],[2703,1],[2704,0],[2713,1],[2714,0],[2723,1],[2724,0],[2733,1],[2734,0],[2743,1],[2744,0],[2753,1],[2754,0],[2763,1],[2793,0],[2823,0],[2883,1],[2914,1],[3033,0]]
     }]
   }]
 }
 {
  type=MovieSceneEventTrack
  sections [{
    type=MovieSceneEventTriggerSection
    channels [None {
      type=MovieSceneScriptingEventChannel
      keys [[2823, "<Struct 'MovieSceneEvent' (0x49e50fe5b40) {}>"], [3033, "<Struct 'MovieSceneEvent' (0x49e50fe5820) {}>"]]
     }]
   }]
 }
]
